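# Landscape material function that paints the Cattus spiral as pink flower coloring on terrain base color and outputs masks for flowers and grass.
# An offset, scaled and rotated UV stencil samples the spiral texture: R sets coverage and desaturation, G blends two tint colors.
# The tint is limited by rock slope and the Cattus landscape layer, while FlowerMask ignores that layer and is subtracted from GrassMask.
# A separate output tints an input color by vertex-color mask.
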MaterialFunction /Game/Athena/Environments/Cattus/MaterialFunctions/MF_CattusSpiral
graph {nodes [
  Ceil()
  n1 = GetMaterialAttributes(None=FunctionInput() {inputType=FunctionInput_MaterialAttributes}) {attributeGetTypes [#1]}
  n2 = TextureCoordinate()
  n3 = AppendVector(
   A=ScalarParameter() {defaultValue=-791.5908, parameterName=OffsetCattusStencil_x}
   B=ScalarParameter() {defaultValue=-573.616, parameterName=OffsetCattusStencil_y}
  )
  n4 = Rotator(
   Coordinate=Divide(A=Add(A=@n2, B=@n3), B=ScalarParameter() {defaultValue=100, parameterName=ScaleCattusStencil})
   Time=ScalarParameter() {defaultValue=10.3433, parameterName=RotationCattusStencil}
  )
  n5 = TextureSample(UVs=@n4) {samplerType=SAMPLERTYPE_Masks, texture=/Game/Athena/Environments/Cattus/Textures/CattusSpiral_M}
  n6 = VectorParameter() {parameterName=ColorCattusFlowers1, defaultValue=<rgba 2 0.3916 0.8355 1>}
  n7 = VectorParameter() {parameterName=ColorCattusFlowers2, defaultValue=<rgba 5 0.6758 1.3251 1>}
  n8 = FunctionInput() {bUsePreviewValueAsDefault=true, inputName=Rock_SlopeMask, inputType=FunctionInput_Scalar, sortPriority=1, previewValue=<1 0 0 1>}
  n9 = LinearInterpolate(
   A=@n1.BaseColor
   B=Multiply(A=Desaturation(None=@n1.BaseColor, Fraction=@n5.R), B=LinearInterpolate(A=@n6.RGB, B=@n7.RGB, Alpha=@n5.G))
   Alpha=Multiply(A=@n5.R, B=@n8)
  )
  n10 = LandscapeLayerSample() {parameterName=Cattus}
  FunctionOutput(None=SetMaterialAttributes!(MaterialAttributes=@n1.MaterialAttributes, BaseColor=LinearInterpolate(A=@n1.BaseColor, B=@n9, Alpha=@n10)) {attributeSetTypes [#1]}) {outputName=Material Out}
  n11 = Multiply(A=Multiply(A=@n5.R, B=@n8)) {constB=2}
  FunctionOutput(None=@n11) {outputName=FlowerMask, sortPriority=2}
  FunctionOutput(None=Subtract(A=FunctionInput() {inputName=GrassMask, inputType=FunctionInput_Scalar, sortPriority=2}, B=@n11)) {outputName=GrassMask, sortPriority=2}
  n12 = FunctionInput() {inputName=BaseColorTintOnlybyVertColor, sortPriority=8}
  n13 = FunctionInput() {inputName=VertexColorMask, inputType=FunctionInput_Scalar, sortPriority=9}
  n14 = Desaturation(None=@n12, Fraction=@n13)
  n15 = LinearInterpolate(A=@n6.RGB, B=@n7.RGB)
  FunctionOutput(None=LinearInterpolate(A=@n12, B=Multiply(A=@n14, B=@n15), Alpha=@n13)) {outputName=BaseColorOnlyOutput, sortPriority=2}
  LinearInterpolate()
  Multiply(B=SphereMask(
    A=@n4
    B=Constant2Vector() {g=0.5, r=0.5}
    Radius=ScalarParameter() {defaultValue=1, parameterName=smr}
    Hardness=ScalarParameter() {parameterName=smrhard}
   ))
  n16 = ScalarParameter() {defaultValue=10, parameterName=ContrastCattusGrassMask1}
  n17 = ScalarParameter() {defaultValue=-71.7639, parameterName=ContrastCattusGrassMask2}
  n18 = ComponentMask(None=@n1.BaseColor) {g=true}
  Multiply(A=Saturate(None=LinearInterpolate(A=@n16, B=@n17, Alpha=@n18)), B=@n5.R)
  VertexNormalWS()
 ]}
